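# Editor utility widget that forms a category header row in the NeverCook context menu.
# A transparent border wraps a horizontal box with a small bold grey label, followed by the Icon_Category image.
# The one-unit-high image fills the remaining width, forming a divider line beside the label.
# The label and image are exposed as widget variables, and the event graph holds only unconnected stub events.
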WidgetBlueprint /Game/NeverCook/UI/ContextMenu/EUW_ContextMenu_CategoryItem
graphs [EventGraph {nodes [
   Event PreConstruct()
   Event Construct()
   Event Tick()
  ]}]
variables []
widgets {
 info {inheritedWidgetCount=0, parentClass=&EditorUtilityWidget, rootWidgetClass=&Border, widgetCount=4}
 widgets [
  Border_0 {widgetClassPath=&Border, widget=Border_0, properties {brushColor=<rgba 1 1 1 0>, padding=<ltrb 0 0 0 0>}}
  HorizontalBox_0 {
   widgetClassPath=&HorizontalBox
   parent=Border_0
   widget=HorizontalBox_0
   slot {class=BorderSlot, properties {padding=<ltrb 0 0 0 0>}}
  }
  ButtonText {
   widgetClassPath=&TextBlock
   bIsVariable=true
   parent=HorizontalBox_0
   widget=ButtonText
   properties {
    text=CATEGORY
    colorAndOpacity {colorUseRule=UseColor_Specified, specifiedColor=<rgba 0.2353 0.2353 0.2353 1>}
    font {
     bForceMonospaced=false
     bMaterialIsStencil=false
     fontMaterial=null
     fontObject=&Roboto
     letterSpacing=0
     monospacedWidth=1
     size=8
     skewAmount=0
     typefaceFontName=Bold
     outlineSettings {
      bApplyOutlineToDropShadows=false
      bMiteredCorners=false
      bSeparateFillAlpha=false
      outlineMaterial=null
      outlineSize=0
      outlineColor=<rgba 0 0 0 1>
     }
    }
   }
   slot {class=HorizontalBoxSlot, properties {horizontalAlignment=HAlign_Left, verticalAlignment=VAlign_Center, padding=<ltrb 10 5 0 5>}}
  }
  Icon {
   widgetClassPath=&Image
   bIsVariable=true
   parent=HorizontalBox_0
   widget=Icon
   properties {brush {
     bIsDynamicallyLoaded=false
     drawAs=Image
     imageType=NoImage
     mirroring=NoMirror
     resourceName=None
     resourceObject=/Game/NeverCook/UI/ContextMenu/Icon/Icon_Category
     tiling=NoTile
     imageSize=<1 1>
     margin=<ltrb 0 0 0 0>
     outlineSettings {
      bUseBrushTransparency=false
      roundingType=HalfHeightRadius
      width=0
      color {colorUseRule=UseColor_Specified, specifiedColor=<rgba 0 0 0 0>}
      cornerRadii=<0 0 0 0>
     }
     tintColor {colorUseRule=UseColor_Specified, specifiedColor=<rgba 1 1 1 1>}
     uVRegion {bIsValid=false, max=<0 0>, min=<0 0>}
    }}
   slot {class=HorizontalBoxSlot, properties {verticalAlignment=VAlign_Center, padding=<ltrb 10 0 0 0>, size {sizeRule=Fill, value=1}}}
  }
 ]
}
