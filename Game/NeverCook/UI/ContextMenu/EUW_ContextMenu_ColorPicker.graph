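# An editor utility widget that serves as a color picker item for a context menu.
# Its only widget is a Details View filtered to the Color property, pointed at the widget itself on PreConstruct and Construct, so the engine's property editor acts as the picker.
# A property change in that view broadcasts Context Menu Item Clicked with this widget and its Slot Widget reference.
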WidgetBlueprint /Game/NeverCook/UI/ContextMenu/EUW_ContextMenu_ColorPicker
graphs [
 Context Menu Item Clicked {nodes [Context Menu Item Clicked!() -> (then, Context Item:Editor Utility Widget Object Reference, Slot Widget:Editor Utility Widget Object Reference) {is=title K2Node_FunctionEntry}]}
 EventGraph {nodes [
   "On Property Changed (ColorVIew)"!() -> (OutputDelegate, then, PropertyName:Name=None) {is=title K2Node_ComponentBoundEvent}
   > Call Context Menu Item Clicked!(
    execute:Exec
    self:EUW Context Menu Color Picker Object Reference
    Context Item:Editor Utility Widget Object Reference=Self-Reference()
    Slot Widget:Editor Utility Widget Object Reference=Get Slot Widget() -> (Slot Widget:Editor Utility Widget Object Reference)
   ) -> (then:Exec) {is=title K2Node_CallDelegate}
   n1 = Event PreConstruct()
   SetObject(execute=[@n1.then, @n2.then], self=Get ColorVIew() -> (ColorVIew:Details View Object Reference), NewObject=Self-Reference())
   n2 = Event Construct()
   Event Tick()
  ]}
]
variables [
 color {type=LinearColor, default=<rgba 0.807 0.807 0.807 1>}
 Context Menu {type=EUW_ContextMenu_C, default=null}
 Context Menu Item Clicked {}
 Owner {type=UserWidget, default=null}
 Slot Widget {type=EditorUtilityWidget, default=null}
]
widgets {
 info {
  inheritedWidgetCount=0
  parentClass=&EditorUtilityWidget
  rootWidgetClass=&DetailsView
  widgetCount=1
 }
 widgets [ColorVIew {
   widgetClassPath=&DetailsView
   widget=ColorVIew
   properties {bShowScrollBar=false, propertiesToShow [Color]}
   slot=~
  }]
}
