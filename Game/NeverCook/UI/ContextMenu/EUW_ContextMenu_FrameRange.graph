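# Editor Utility Widget for a frame-range row, placed among context menu UI assets.
# A horizontal box holds FrameStart and FrameEnd Fortnite Creative numeric entries around a "To" label, both slider-less, clamped to ±9999 and exposed as variables.
# The Blueprint has no logic: its events are empty, and its clicked dispatcher with a Text string and its String Data variable are unused here.
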WidgetBlueprint /Game/NeverCook/UI/ContextMenu/EUW_ContextMenu_FrameRange
graphs [
 Context Menu Item Clicked_0 {nodes [Context Menu Item Clicked_0!() -> (then, Text:String) {is=title K2Node_FunctionEntry}]}
 EventGraph {nodes [
   Event PreConstruct()
   Event Construct()
   Event Tick()
  ]}
]
variables [
 Context Menu Item Clicked_0 {}
 String Data_0 {type=string, default=""}
]
widgets {
 info {inheritedWidgetCount=0, parentClass=&EditorUtilityWidget, rootWidgetClass=&HorizontalBox, widgetCount=4}
 widgets [
  HorizontalBox_1 {widgetClassPath=&HorizontalBox, widget=HorizontalBox_1}
  FrameStart {
   widgetClassPath=/Script/FortniteUI.FortCreativePropertyNumericEntry
   bIsVariable=true
   parent=HorizontalBox_1
   widget=FrameStart
   properties {
    bEnableSlider=false
    bOverride_MaxSliderValue=true
    bOverride_MaxValue=true
    bOverride_MinSliderValue=true
    bOverride_MinValue=true
    clearKeyboardFocusOnCommit=true
    delta=1
    justification=Right
    maxFractionalDigits=2
    maxSliderValue=9999
    maxValue=9999
    minDesiredWidth=56
    minFractionalDigits=0
    minSliderValue=-9999
    minValue=-9999
    foregroundColor {colorUseRule=UseColor_Specified, specifiedColor=<rgba 0.9062 0.9062 0.9062 1>}
    widgetStyle {
     activeBackgroundBrush {
      bIsDynamicallyLoaded=false
      drawAs=RoundedBox
      imageType=FullColor
      mirroring=NoMirror
      resourceName=../../../Engine/Content/Slate/Common/Spinbox.png
      resourceObject=null
      tiling=NoTile
      imageSize=<0 0>
      margin=<ltrb 0.25 0.25 0.25 0.25>
      outlineSettings {
       bUseBrushTransparency=false
       roundingType=FixedRadius
       width=1
       color {colorUseRule=UseColor_Specified, specifiedColor=<rgba 1 1 1 0.3739>}
       cornerRadii=<4 4 4 4>
      }
      tintColor {colorUseRule=UseColor_Specified, specifiedColor=<rgba 0.0417 0.0417 0.0417 1>}
      uVRegion {bIsValid=false, max=<0 0>, min=<0 0>}
     }
     activeFillBrush {
      bIsDynamicallyLoaded=false
      drawAs=Box
      imageType=FullColor
      mirroring=NoMirror
      resourceName=../../../Engine/Content/Slate/Common/Spinbox_Fill.png
      resourceObject=null
      tiling=NoTile
      imageSize=<0 0>
      margin=<ltrb 0.25 0.25 0.5 0.25>
      outlineSettings {
       bUseBrushTransparency=false
       roundingType=HalfHeightRadius
       width=0
       color {colorUseRule=UseColor_Specified, specifiedColor=<rgba 0 0 0 0>}
       cornerRadii=<0 0 0 0>
      }
      tintColor {colorUseRule=UseColor_Specified, specifiedColor=<rgba 0.0898 0.1256 0.224 0>}
      uVRegion {bIsValid=false, max=<0 0>, min=<0 0>}
     }
     arrowsImage {
      bIsDynamicallyLoaded=false
      drawAs=Image
      imageType=FullColor
      mirroring=NoMirror
      resourceName=../../../Engine/Content/Slate/Common/SpinArrows.png
      resourceObject=null
      tiling=NoTile
      imageSize=<12 12>
      margin=<ltrb 0 0 0 0>
      outlineSettings {
       bUseBrushTransparency=false
       roundingType=HalfHeightRadius
       width=0
       color {colorUseRule=UseColor_Specified, specifiedColor=<rgba 0 0 0 0>}
       cornerRadii=<0 0 0 0>
      }
      tintColor {colorUseRule=UseColor_Specified, specifiedColor=<rgba 1 1 1 1>}
      uVRegion {bIsValid=false, max=<0 0>, min=<0 0>}
     }
     backgroundBrush {
      bIsDynamicallyLoaded=false
      drawAs=RoundedBox
      imageType=FullColor
      mirroring=NoMirror
      resourceName=../../../Engine/Content/Slate/Common/Spinbox.png
      resourceObject=null
      tiling=NoTile
      imageSize=<0 0>
      margin=<ltrb 0.25 0.25 0.25 0.25>
      outlineSettings {
       bUseBrushTransparency=false
       roundingType=FixedRadius
       width=0
       color {colorUseRule=UseColor_Specified, specifiedColor=<rgba 1 1 1 0>}
       cornerRadii=<4 4 4 4>
      }
      tintColor {colorUseRule=UseColor_Specified, specifiedColor=<rgba 0.0417 0.0417 0.0417 1>}
      uVRegion {bIsValid=false, max=<0 0>, min=<0 0>}
     }
     foregroundColor {colorUseRule=UseColor_Specified, specifiedColor=<rgba 0.4729 0.4729 0.4729 1>}
     hoveredBackgroundBrush {
      bIsDynamicallyLoaded=false
      drawAs=RoundedBox
      imageType=FullColor
      mirroring=NoMirror
      resourceName=../../../Engine/Content/Slate/Common/Spinbox.png
      resourceObject=null
      tiling=NoTile
      imageSize=<0 0>
      margin=<ltrb 0.25 0.25 0.25 0.25>
      outlineSettings {
       bUseBrushTransparency=false
       roundingType=FixedRadius
       width=1
       color {colorUseRule=UseColor_Specified, specifiedColor=<rgba 1 1 1 0.3739>}
       cornerRadii=<4 4 4 4>
      }
      tintColor {colorUseRule=UseColor_Specified, specifiedColor=<rgba 0.0417 0.0417 0.0417 1>}
      uVRegion {bIsValid=false, max=<0 0>, min=<0 0>}
     }
     hoveredFillBrush {
      bIsDynamicallyLoaded=false
      drawAs=NoDrawType
      imageType=NoImage
      mirroring=NoMirror
      resourceName=None
      resourceObject=null
      tiling=NoTile
      imageSize=<0 0>
      margin=<ltrb 0 0 0 0>
      outlineSettings {
       bUseBrushTransparency=false
       roundingType=HalfHeightRadius
       width=0
       color {colorUseRule=UseColor_Specified, specifiedColor=<rgba 0 0 0 0>}
       cornerRadii=<0 0 0 0>
      }
      tintColor {colorUseRule=UseColor_Specified, specifiedColor=<rgba 0.1618 0.5104 0.4579 0>}
      uVRegion {bIsValid=false, max=<0 0>, min=<0 0>}
     }
     inactiveFillBrush {
      bIsDynamicallyLoaded=false
      drawAs=Box
      imageType=FullColor
      mirroring=NoMirror
      resourceName=../../../Engine/Content/Slate/Common/Spinbox_Fill_Hovered.png
      resourceObject=null
      tiling=NoTile
      imageSize=<0 0>
      margin=<ltrb 0.25 0.25 0.25 0.25>
      outlineSettings {
       bUseBrushTransparency=false
       roundingType=HalfHeightRadius
       width=0
       color {colorUseRule=UseColor_Specified, specifiedColor=<rgba 0 0 0 0>}
       cornerRadii=<0 0 0 0>
      }
      tintColor {colorUseRule=UseColor_Specified, specifiedColor=<rgba 0.0397 0.125 0.1109 0>}
      uVRegion {bIsValid=false, max=<0 0>, min=<0 0>}
     }
     insetPadding=<ltrb 0 0 0 0>
     textPadding=<ltrb 1 2 1 2>
    }
   }
   slot {class=HorizontalBoxSlot, properties {verticalAlignment=VAlign_Center, size {sizeRule=Fill, value=1}}}
  }
  ButtonText {
   widgetClassPath=&TextBlock
   parent=HorizontalBox_1
   widget=ButtonText
   properties {
    text=To
    colorAndOpacity {colorUseRule=UseColor_Specified, specifiedColor=<rgba 1 1 1 0.5>}
    font {
     bForceMonospaced=false
     bMaterialIsStencil=false
     fontMaterial=null
     fontObject=&Roboto
     letterSpacing=0
     monospacedWidth=1
     size=12
     skewAmount=0
     typefaceFontName=Regular
     outlineSettings {
      bApplyOutlineToDropShadows=false
      bMiteredCorners=false
      bSeparateFillAlpha=false
      outlineMaterial=null
      outlineSize=0
      outlineColor=<rgba 0 0 0 1>
     }
    }
   }
   slot {class=HorizontalBoxSlot, properties {verticalAlignment=VAlign_Center, padding=<ltrb 8 0 8 0>}}
  }
  FrameEnd {
   widgetClassPath=&FortCreativePropertyNumericEntry
   bIsVariable=true
   parent=HorizontalBox_1
   widget=FrameEnd
   properties {
    bEnableSlider=false
    bOverride_MaxSliderValue=true
    bOverride_MaxValue=true
    bOverride_MinSliderValue=true
    bOverride_MinValue=true
    clearKeyboardFocusOnCommit=true
    delta=1
    justification=Right
    maxFractionalDigits=2
    maxSliderValue=9999
    maxValue=9999
    minDesiredWidth=56
    minFractionalDigits=0
    minSliderValue=-9999
    minValue=-9999
    foregroundColor {colorUseRule=UseColor_Specified, specifiedColor=<rgba 0.9062 0.9062 0.9062 1>}
    widgetStyle {
     activeBackgroundBrush {
      bIsDynamicallyLoaded=false
      drawAs=RoundedBox
      imageType=FullColor
      mirroring=NoMirror
      resourceName=../../../Engine/Content/Slate/Common/Spinbox.png
      resourceObject=null
      tiling=NoTile
      imageSize=<0 0>
      margin=<ltrb 0.25 0.25 0.25 0.25>
      outlineSettings {
       bUseBrushTransparency=false
       roundingType=FixedRadius
       width=1
       color {colorUseRule=UseColor_Specified, specifiedColor=<rgba 1 1 1 0.3739>}
       cornerRadii=<4 4 4 4>
      }
      tintColor {colorUseRule=UseColor_Specified, specifiedColor=<rgba 0.0417 0.0417 0.0417 1>}
      uVRegion {bIsValid=false, max=<0 0>, min=<0 0>}
     }
     activeFillBrush {
      bIsDynamicallyLoaded=false
      drawAs=Box
      imageType=FullColor
      mirroring=NoMirror
      resourceName=../../../Engine/Content/Slate/Common/Spinbox_Fill.png
      resourceObject=null
      tiling=NoTile
      imageSize=<0 0>
      margin=<ltrb 0.25 0.25 0.5 0.25>
      outlineSettings {
       bUseBrushTransparency=false
       roundingType=HalfHeightRadius
       width=0
       color {colorUseRule=UseColor_Specified, specifiedColor=<rgba 0 0 0 0>}
       cornerRadii=<0 0 0 0>
      }
      tintColor {colorUseRule=UseColor_Specified, specifiedColor=<rgba 0.0898 0.1256 0.224 0>}
      uVRegion {bIsValid=false, max=<0 0>, min=<0 0>}
     }
     arrowsImage {
      bIsDynamicallyLoaded=false
      drawAs=Image
      imageType=FullColor
      mirroring=NoMirror
      resourceName=../../../Engine/Content/Slate/Common/SpinArrows.png
      resourceObject=null
      tiling=NoTile
      imageSize=<12 12>
      margin=<ltrb 0 0 0 0>
      outlineSettings {
       bUseBrushTransparency=false
       roundingType=HalfHeightRadius
       width=0
       color {colorUseRule=UseColor_Specified, specifiedColor=<rgba 0 0 0 0>}
       cornerRadii=<0 0 0 0>
      }
      tintColor {colorUseRule=UseColor_Specified, specifiedColor=<rgba 1 1 1 1>}
      uVRegion {bIsValid=false, max=<0 0>, min=<0 0>}
     }
     backgroundBrush {
      bIsDynamicallyLoaded=false
      drawAs=RoundedBox
      imageType=FullColor
      mirroring=NoMirror
      resourceName=../../../Engine/Content/Slate/Common/Spinbox.png
      resourceObject=null
      tiling=NoTile
      imageSize=<0 0>
      margin=<ltrb 0.25 0.25 0.25 0.25>
      outlineSettings {
       bUseBrushTransparency=false
       roundingType=FixedRadius
       width=0
       color {colorUseRule=UseColor_Specified, specifiedColor=<rgba 1 1 1 0>}
       cornerRadii=<4 4 4 4>
      }
      tintColor {colorUseRule=UseColor_Specified, specifiedColor=<rgba 0.0417 0.0417 0.0417 1>}
      uVRegion {bIsValid=false, max=<0 0>, min=<0 0>}
     }
     foregroundColor {colorUseRule=UseColor_Specified, specifiedColor=<rgba 0.4729 0.4729 0.4729 1>}
     hoveredBackgroundBrush {
      bIsDynamicallyLoaded=false
      drawAs=RoundedBox
      imageType=FullColor
      mirroring=NoMirror
      resourceName=../../../Engine/Content/Slate/Common/Spinbox.png
      resourceObject=null
      tiling=NoTile
      imageSize=<0 0>
      margin=<ltrb 0.25 0.25 0.25 0.25>
      outlineSettings {
       bUseBrushTransparency=false
       roundingType=FixedRadius
       width=1
       color {colorUseRule=UseColor_Specified, specifiedColor=<rgba 1 1 1 0.3739>}
       cornerRadii=<4 4 4 4>
      }
      tintColor {colorUseRule=UseColor_Specified, specifiedColor=<rgba 0.0417 0.0417 0.0417 1>}
      uVRegion {bIsValid=false, max=<0 0>, min=<0 0>}
     }
     hoveredFillBrush {
      bIsDynamicallyLoaded=false
      drawAs=NoDrawType
      imageType=NoImage
      mirroring=NoMirror
      resourceName=None
      resourceObject=null
      tiling=NoTile
      imageSize=<0 0>
      margin=<ltrb 0 0 0 0>
      outlineSettings {
       bUseBrushTransparency=false
       roundingType=HalfHeightRadius
       width=0
       color {colorUseRule=UseColor_Specified, specifiedColor=<rgba 0 0 0 0>}
       cornerRadii=<0 0 0 0>
      }
      tintColor {colorUseRule=UseColor_Specified, specifiedColor=<rgba 0.1618 0.5104 0.4579 0>}
      uVRegion {bIsValid=false, max=<0 0>, min=<0 0>}
     }
     inactiveFillBrush {
      bIsDynamicallyLoaded=false
      drawAs=Box
      imageType=FullColor
      mirroring=NoMirror
      resourceName=../../../Engine/Content/Slate/Common/Spinbox_Fill_Hovered.png
      resourceObject=null
      tiling=NoTile
      imageSize=<0 0>
      margin=<ltrb 0.25 0.25 0.25 0.25>
      outlineSettings {
       bUseBrushTransparency=false
       roundingType=HalfHeightRadius
       width=0
       color {colorUseRule=UseColor_Specified, specifiedColor=<rgba 0 0 0 0>}
       cornerRadii=<0 0 0 0>
      }
      tintColor {colorUseRule=UseColor_Specified, specifiedColor=<rgba 0.0397 0.125 0.1109 0>}
      uVRegion {bIsValid=false, max=<0 0>, min=<0 0>}
     }
     insetPadding=<ltrb 0 0 0 0>
     textPadding=<ltrb 1 2 1 2>
    }
   }
   slot {class=HorizontalBoxSlot, properties {verticalAlignment=VAlign_Center, size {sizeRule=Fill, value=1}}}
  }
 ]
}
