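# EUW_ContextMenu is an Editor Utility Widget for a reusable, scrollable context menu grouped by category.
# Add stores items by category and name, wrapping a button or custom slot widget in an item container bound to the caller's delegate.
# BuildContent fills the VerticalBox with uppercase category headers and their items.
# Position Menu either anchors to the top parent's bottom-left or uses inline Python to offset from the mouse within bounds; Show is empty.
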EditorUtilityWidgetBlueprint /Game/NeverCook/UI/ContextMenu/EUW_ContextMenu
graphs [
 Add {nodes [
   n1 = Add!() -> (
    then
    Event:"Delegate (by ref)"
    Category:String
    Name:String
    Icon:Texture 2D Object Reference
    ItemWidget:User Widget Class Reference
    Tooltip:String
   ) {is=title K2Node_FunctionEntry}
   > n2 = Sequence!(execute) -> (then_0, then_1, then_2:Exec, then_3:Exec, then_4:Exec, then_5:Exec)
   n3 = Get MenuCategories() -> (MenuCategories:Map of Strings to Struct Context Menu Category Item Structures)
   > n4 = Branch(Condition=Contains!(TargetMap:Map of Strings to Struct Context Menu Category Item Structures=@n3, Key:"String (by ref)"=@n1.Category) -> (ReturnValue))
   n5 = Set NewLocalVar(
    execute=@n4.else
    NewLocalVar:Map of Strings to Struct Context Menu Item Structures=Make Map!() -> (Map:Map of Strings to Struct Context Menu Item Structures) {is=title K2Node_MakeMap}
   ) -> (Output_Get:Map of Strings to Struct Context Menu Item Structures)
   > n6 = Set CurrentCategory(CurrentCategory:Struct Context Menu Category Item Structure=Make Struct Context Menu Category Item!(
     Category_2_#1:String=@n1.Category
     Items_9_#2:Map of Strings to Struct Context Menu Item Structures=@n5
    ) -> (Struct_ContextMenu_CategoryItem:Struct Context Menu Category Item Structure) {is=title K2Node_MakeStruct}) -> (Output_Get:Struct Context Menu Category Item Structure)
   n7 = Find(TargetMap:Map of Strings to Struct Context Menu Category Item Structures=@n3, Key=@n1.Category) -> (Value:Struct Context Menu Category Item Structure)
   n8 = Get NewLocalVar() -> (NewLocalVar:Map of Strings to Struct Context Menu Item Structures)
   n9 = Set members in Struct Context Menu Category Item!(
    execute:Exec=@n2.then_4
    StructRef:"Struct Context Menu Category Item Structure (by ref)"=@n7
    Category_2_#1:String=@n1.Category
    Items_9_#2:Map of Strings to Struct Context Menu Item Structures=@n8
   ) -> (then:Exec, StructOut:"Struct Context Menu Category Item Structure (by ref)") {is=title K2Node_SetFieldsInStruct}
   Add!(
    execute=[@n9.then, @n6.then]
    TargetMap:Map of Strings to Struct Context Menu Category Item Structures=@n3
    Key:"String (by ref)"=@n1.Category
    Value:"Struct Context Menu Category Item Structure (by ref)"=@n9
   ) -> (then) {is=title K2Node_CallFunction}
   n10 = Set CurrentCategory!(execute=@n2.then_1, CurrentCategory=@n7) -> (
    then
    Output_Get_Category_2_#1:String
    Output_Get_Items_9_#2:Map of Strings to Struct Context Menu Item Structures
   )
   > Set NewLocalVar(NewLocalVar=@n10.Output_Get_Items_9_#2)
   n11 = Select(
    Option 0:User Widget Class Reference=&/EUW_ContextMenu_Button.EUW_ContextMenu_Button_C
    Option 1:User Widget Class Reference=@n1.ItemWidget
    Index=IsValidClass!(Class:Object Class Reference=@n1.ItemWidget) -> (ReturnValue:bool)
   ) -> (ReturnValue:User Widget Class Reference)
   n12 = Add Icon Item!(
    execute=@n2.then_2
    self
    Event:"Delegate (by ref)"=@n1
    Icon:Texture 2D Object Reference=@n1.Icon
    Slot Widget:User Widget Class Reference=@n11
    ItemName:String=@n1.Name
   ) -> (then, Context Menu Widget:Widget Object Reference, Item Container:User Widget Object Reference)
   > Add(
    TargetMap:Map of Strings to Struct Context Menu Item Structures=@n8
    Key=@n1.Name
    Value:"Struct Context Menu Item Structure (by ref)"=Make Struct Context Menu Item!(
     Name_2_#3:String=@n1.Name
     Category_4_#4:String=@n1.Category
     Icon_8_#5:Texture 2D Object Reference=@n1.Icon
     ItemWidgetClass_25_#6:User Widget Class Reference=@n11
     Widget_24_#7:Widget Object Reference=@n12.Item Container
    ) -> (Struct_ContextMenu_Item:Struct Context Menu Item Structure) {is=title K2Node_MakeStruct}
   )
   Set ToolTipText!(execute=@n2.then_3, ToolTipText:Text="To Text (String)"(InString=@n1.Tooltip), self:Widget Object Reference=@n12) -> (then, Output_Get:Text)
   Return Node!(execute=@n2.then_5, Context Menu Item:Widget Object Reference=@n12, ContainerItem:User Widget Object Reference=@n12.Item Container)
  ]}
 Add Icon Item {nodes [
   n1 = Add Icon Item!() -> (then, Event:"Delegate (by ref)", Icon:Texture 2D Object Reference, Slot Widget:User Widget Class Reference, ItemName:String) {is=title K2Node_FunctionEntry}
   > n2 = Add Item!(execute, self, Slot Widget:User Widget Class Reference=@n1.Slot Widget) -> (then, Menu Item Container:EUW Context Menu Item Object Reference, Widget:Editor Utility Widget Object Reference)
   > Bind Event to Context Menu Item Clicked(self:EUW Context Menu Item Object Reference=@n2, Delegate=@n1) {is=title K2Node_AddDelegate}
   > n3 = Is Valid(InputObject=@n1.Icon)
   n4 = Branch(
    execute=[@n17.then, @n3.Is Not Valid]
    Condition="Equal (Class)"(A=@n1.Slot Widget, B=&EUW_ContextMenu_Button_C)
   )
   > n5 = Cast To EUW_ContextMenu_Button!(execute, Object=@n2.Widget) -> (then, CastFailed, AsEUW Context Menu Button:EUW Context Menu Button Object Reference) {is=title K2Node_DynamicCast}
   > Bind Event to Context Menu Item Clicked(self:EUW Context Menu Button Object Reference=@n5, Delegate=@n1)
   n6 = "To Text (String)"(InString=Get ItemName() -> (ItemName:String))
   > n7 = Set Text(Text=@n6, self=Get ButtonText!(self:EUW Context Menu Button Object Reference=@n5) -> (ButtonText:Text Object Reference))
   Return Node!(
    execute=[@n15.then, @n9.then, @n12.then, @n7.then]
    Context Menu Widget:Widget Object Reference=@n2.Widget
    Item Container:User Widget Object Reference=@n2
   )
   Return Node(execute=[@n9.CastFailed, @n5.CastFailed, @n11.CastFailed, @n13.else])
   n8 = Branch(
    execute=@n4.else
    Condition="Equal (Class)"(A=@n1.Slot Widget, B=&/EUW_ContextMenu_FrameRange.EUW_ContextMenu_FrameRange_C)
   )
   > n9 = Cast To EUW_ContextMenu_FrameRange!(execute, Object=@n2.Widget) -> (then, CastFailed, AsEUW Context Menu Frame Range:EUW Context Menu Frame Range Object Reference) {is=title K2Node_DynamicCast}
   n10 = Branch(
    execute=@n8.else
    Condition="Equal (Class)"(A=@n1.Slot Widget, B=&/EUW_ContextMenu_CheckBox.EUW_ContextMenu_CheckBox_C)
   )
   > n11 = Cast To EUW_ContextMenu_CheckBox!(execute, Object=@n2.Widget) -> (then, CastFailed, AsEUW Context Menu Check Box:EUW Context Menu Check Box Object Reference) {is=title K2Node_DynamicCast}
   > n12 = Set Text(Text=@n6, self=Get ButtonText(self:EUW Context Menu Check Box Object Reference=@n11))
   n13 = Branch(
    execute=@n10.else
    Condition="Equal (Class)"(A=@n1.Slot Widget, B=&/EUW_ContextMenu_ColorPicker.EUW_ContextMenu_ColorPicker_C)
   )
   > n14 = Cast To EUW_ContextMenu_ColorPicker!(execute, Object=@n2.Widget) -> (then, CastFailed, AsEUW Context Menu Color Picker:EUW Context Menu Color Picker Object Reference) {is=title K2Node_DynamicCast}
   > n15 = Bind Event to Context Menu Item Clicked(self:EUW Context Menu Color Picker Object Reference=@n14, Delegate=@n1)
   n16 = Get Icon!(self:EUW Context Menu Item Object Reference=@n2) -> (Icon:Image Object Reference)
   SetBrush(
    execute=@n3.Is Valid
    self=@n16
    InBrush=MakeBrushFromTexture!(Texture:Texture 2D Object Reference=@n1.Icon, Width:int=16, Height:int=16) -> (ReturnValue:Slate Brush Structure)
   )
   > n17 = SetVisibility(self=@n16, InVisibility=Visible)
  ]}
 Add Item {nodes [
   Add Item!() -> (then, Slot Widget:User Widget Class Reference) {is=title K2Node_FunctionEntry}
   > n1 = Create Widget!(
    execute:Exec
    WorldContextObject:Object Reference
    Class:User Widget Class Reference=&/EUW_ContextMenu_Item.EUW_ContextMenu_Item_C
    OwningPlayer:Player Controller Object Reference
   ) -> (then:Exec, ReturnValue:EUW Context Menu Item Object Reference)
   > Set Context Menu!(execute, Context Menu:EUW Context Menu Object Reference=Self-Reference(), self:EUW Context Menu Item Object Reference=@n1) -> (then, Output_Get:EUW Context Menu Object Reference)
   > n2 = Create Widget(Class=Get Slot Widget() -> (Slot Widget:User Widget Class Reference)) -> (ReturnValue:User Widget Object Reference)
   > AddChild(self=Get NamedSlot!(self:EUW Context Menu Item Object Reference=@n1) -> (NamedSlot:Named Slot Object Reference), Content=@n2)
   > n3 = Cast To EditorUtilityWidget!(execute, Object=@n2) -> (then, CastFailed, AsEditor Utility Widget:Editor Utility Widget Object Reference) {is=title K2Node_DynamicCast}
   > Set Slot Widget!(execute, Slot Widget:Editor Utility Widget Object Reference=@n3, self:EUW Context Menu Item Object Reference=@n1) -> (then, Output_Get:Editor Utility Widget Object Reference)
   > Return Node!(execute, Menu Item Container:EUW Context Menu Item Object Reference=@n1, Widget:Editor Utility Widget Object Reference=@n3)
  ]}
 BuildContent {nodes [
   BuildContent() {is=title K2Node_FunctionEntry}
   > ClearChildren(self=Get VerticalBox() -> (VerticalBox:Vertical Box Object Reference))
   n1 = Get MenuCategories()
   > n2 = Keys!(execute, TargetMap:Map of Strings to Struct Context Menu Category Item Structures=@n1) -> (then, Keys:Array of Strings)
   > n3 = For Each Loop(Array=@n2)
   > n4 = Sequence!(execute) -> (then_0, then_1)
   > n5 = Create Widget(Class=&/EUW_ContextMenu_CategoryItem.EUW_ContextMenu_CategoryItem_C) -> (ReturnValue:EUW Context Menu Category Item Object Reference)
   n6 = Get VerticalBox()
   > AddChild(self=@n6, Content=@n5)
   n7 = Find!(TargetMap=@n1, Key=@n3) -> (
    Value_Category_2_#1:String
    Value_Items_9_#2:Map of Strings to Struct Context Menu Item Structures
    ReturnValue
   )
   > Set Text(
    Text="To Text (String)"(InString=ToUpper!(SourceString:String=@n7) -> (ReturnValue:String))
    self=Get ButtonText(self:EUW Context Menu Category Item Object Reference=@n5)
   )
   n8 = Keys(execute=@n4.then_1, TargetMap:Map of Strings to Struct Context Menu Item Structures=@n7.Value_Items_9_#2)
   > n9 = For Each Loop(Array=@n8)
   > AddChild(
    self=@n6
    Content=Find!(TargetMap:Map of Strings to Struct Context Menu Item Structures=@n7.Value_Items_9_#2, Key=@n9) -> (
     Value_Name_2_#3:String
     Value_Category_4_#4:String
     Value_Icon_8_#5:Texture 2D Object Reference
     Value_ItemWidgetClass_25_#6:User Widget Class Reference
     Value_Widget_24_#7:Widget Object Reference
     ReturnValue
    ).Value_Widget_24_#7
   )
  ]}
 Clear Context Menu {nodes [
   Clear Context Menu() {is=title K2Node_FunctionEntry}
   > Clear!(execute, TargetMap:Map of Strings to Struct Context Menu Category Item Structures=Get MenuCategories()) -> (then) {is=title K2Node_CallFunction}
   > ClearChildren(self=Get VerticalBox())
   > Set Name(Name:String) -> (Output_Get:String)
  ]}
 EventGraph {nodes [
   Event PreConstruct()
   Event Construct()
  ]}
 GetTopParent {nodes [
   n1 = GetTopParent!() -> (then, Widget:Widget Object Reference) {is=title K2Node_FunctionEntry}
   n2 = GetParent!(self:Widget Object Reference=@n1) -> (ReturnValue:Panel Widget Object Reference)
   > n3 = Is Valid(InputObject=@n2)
   Return Node!(execute=@n3.Is Not Valid, TopParent:Widget Object Reference=@n1)
   > n4 = GetTopParent!(execute, self, Widget:Widget Object Reference=@n2) -> (then, TopParent:Widget Object Reference) {is=title K2Node_CallFunction}
   > Return Node(TopParent=@n4)
  ]}
 PointerInWidget {nodes [
   n1 = PointerInWidget!() -> (then, TargetWidget:Widget Object Reference, PointerEvent:"Pointer Event Structure (by ref)") {is=title K2Node_FunctionEntry}
   n2 = BreakVector2D(InVec=Get Screen Space Position!(Input:"Pointer Event Structure (by ref)"=@n1.PointerEvent) -> (ReturnValue:Vector 2D Structure) {is=title K2Node_CallFunction})
   n3 = GetTickSpaceGeometry!(self:Widget Object Reference=@n1) -> (ReturnValue:"Geometry Structure (by ref)")
   n4 = BreakVector2D(InVec=GetAbsoluteSize!(Geometry:"Geometry Structure (by ref)"=@n3) -> (ReturnValue:Vector 2D Structure))
   n5 = BreakVector2D(InVec=GetAbsolutePosition!(Geometry:"Geometry Structure (by ref)"=@n3) -> (ReturnValue:Vector 2D Structure))
   n6 = float + float(A=@n5.Y, B=@n4.Y)
   n7 = float + float(A=@n5, B=@n4)
   > Return Node!(
    execute
    IsWithin:bool=AND Boolean!(
     A="float >= float"(A=@n2, B=@n5)
     B="float < float"(A=@n2, B=@n7)
     C:bool="float >= float"(A=@n2.Y, B=@n5.Y)
     D:bool="float < float"(A=@n2.Y, B=@n6)
    ) -> (ReturnValue)
   )
  ]}
 Position Menu {nodes [
   n1 = Position Menu!() -> (then, Mouse Position:Vector 2D Structure) {is=title K2Node_FunctionEntry}
   n2 = Get OwnerWidget() -> (OwnerWidget:Widget Object Reference)
   > n3 = GetTopParent(Widget=@n2)
   > PrintString(InString=Append(A="TopMenuParent  ", B=GetDisplayName(Object=@n3)))
   > n4 = Get!(execute:Exec) -> (then:Exec, else:Exec, FixedPosition:bool)
   > n5 = GetWidgetAbsoluteRect!(execute, Widget:Widget Object Reference=@n3, __WorldContext:Object Reference) -> (then, TopLeft:Vector 2D Structure, BottomRight:Vector 2D Structure, Size:Vector 2D Structure, DebugStringOutput:String)
   n6 = Select(
    Option 0:Widget Object Reference=@n2
    Option 1:Widget Object Reference=GetParent()
    Index=IsNullItem!(ItemObject:Object Reference=@n2, __WorldContext:Object Reference) -> (Is Null Item:bool)
   ) -> (ReturnValue:Widget Object Reference)
   n7 = GetTickSpaceGeometry(self=@n6)
   n8 = BreakVector2D(InVec=@n5.BottomRight)
   n9 = BreakVector2D(InVec=@n5)
   n10 = BreakVector2D(InVec="Absolute to Local (Coordinates)"!(Geometry:"Geometry Structure (by ref)"=@n7, AbsoluteCoordinate:Vector 2D Structure=MakeVector2D(X=@n9, Y=@n8.Y)) -> (ReturnValue:Vector 2D Structure))
   > SetRenderTransform!(
    execute
    self:Widget Object Reference
    InTransform:Widget Transform Structure=Make Widget Transform!(
     Translation_X:double=@n10
     Translation_Y:double=@n10.Y
     Scale:Vector 2D Structure=(X=1,Y=1)
     Shear:Vector 2D Structure=(X=0,Y=0)
     Angle:float=0
    ) -> (WidgetTransform:Widget Transform Structure) {is=title K2Node_MakeStruct}
   ) -> (then)
   n11 = Execute Python Script!(
    execute=@n4.else
    PythonScript="max_bounds_x = owner_pos.x + owner_size.x\r\nmax_bounds_y = owner_pos.y + owner_size.y\r\n\r\nmenu_bounds_x = mouse_pos.x + menu_size.x\r\nmenu_bounds_y = mouse_pos.y + menu_size.y\r\ndx = 0\r\ndy = 0\r\nif menu_bounds_x > max_bounds_x:\r\n    dx = menu_bounds_x - max_bounds_x\r\nif menu_bounds_y > max_bounds_y:\r\n    dy = menu_bounds_y - max_bounds_y"
    menuSize:Vector 2D Structure=GetAbsoluteSize(Geometry=GetTickSpaceGeometry(self=@n6))
    MousePos:Vector 2D Structure=@n1
    OwnerPos:Vector 2D Structure=GetAbsolutePosition(Geometry=@n7)
    OwnerSize:Vector 2D Structure=GetAbsoluteSize(Geometry=@n7)
   ) -> (then, ReturnValue, dx:double=~, dy:double=~) {inputs [menuSize,MousePos,OwnerPos,OwnerSize], outputs [dx,dy]}
   n12 = BreakVector2D(InVec="Absolute to Local (Coordinates)"(Geometry=@n7, AbsoluteCoordinate=@n1))
   > SetRenderTransform(InTransform=Make Widget Transform(Translation_X=float - float(A=@n12, B=@n11.dx), Translation_Y=float - float(A=@n12.Y, B=@n11.dy), Scale=(X=1,Y=1), Shear=(X=0,Y=0), Angle=0))
  ]}
 Show {nodes [Show!() -> (then, Show:bool=~, Mouse Position:Vector 2D Structure, Name:String) {is=title K2Node_FunctionEntry}]}
]
variables [
 FixedPosition {type=boolean}
 MenuCategories {type="map<Struct_ContextMenu_CategoryItem>", default {}}
 Name {type=string, default=""}
 OwnerWidget {type=Widget, default=null}
]
widgets {
 info {rootWidgetClass=&Border, widgetCount=3}
 widgets [
  Border {widgetClassPath=&Border, widget=Border, properties {brushColor=<rgba 0.0395 0.0395 0.0395 1>, padding=<ltrb 8 8 8 8>}}
  ScrollBox_209 {
   widgetClassPath=&ScrollBox
   parent=Border
   widget=ScrollBox_209
   properties {
    widgetBarStyle {
     thickness=8
     draggedThumbImage {
      bIsDynamicallyLoaded=false
      drawAs=RoundedBox
      imageType=NoImage
      mirroring=NoMirror
      resourceName=None
      resourceObject=null
      tiling=NoTile
      imageSize=<0 0>
      margin=<ltrb 0 0 0 0>
      outlineSettings {
       bUseBrushTransparency=false
       roundingType=FixedRadius
       width=0
       color {colorUseRule=UseColor_Specified, specifiedColor=<rgba 0 0 0 0>}
       cornerRadii=<4 4 4 4>
      }
      tintColor {colorUseRule=UseColor_Specified, specifiedColor=<rgba 0.2159 0.2159 0.2159 1>}
      uVRegion {bIsValid=false, max=<0 0>, min=<0 0>}
     }
     horizontalBackgroundImage {
      bIsDynamicallyLoaded=false
      drawAs=NoDrawType
      imageType=NoImage
      mirroring=NoMirror
      resourceName=None
      resourceObject=null
      tiling=NoTile
      imageSize=<0 0>
      margin=<ltrb 0 0 0 0>
      outlineSettings {
       bUseBrushTransparency=false
       roundingType=HalfHeightRadius
       width=0
       color {colorUseRule=UseColor_Specified, specifiedColor=<rgba 0 0 0 0>}
       cornerRadii=<0 0 0 0>
      }
      tintColor {colorUseRule=UseColor_Specified, specifiedColor=<rgba 1 1 1 1>}
      uVRegion {bIsValid=false, max=<0 0>, min=<0 0>}
     }
     horizontalBottomSlotImage {
      bIsDynamicallyLoaded=false
      drawAs=NoDrawType
      imageType=NoImage
      mirroring=NoMirror
      resourceName=None
      resourceObject=null
      tiling=NoTile
      imageSize=<0 0>
      margin=<ltrb 0 0 0 0>
      outlineSettings {
       bUseBrushTransparency=false
       roundingType=HalfHeightRadius
       width=0
       color {colorUseRule=UseColor_Specified, specifiedColor=<rgba 0 0 0 0>}
       cornerRadii=<0 0 0 0>
      }
      tintColor {colorUseRule=UseColor_Specified, specifiedColor=<rgba 1 1 1 1>}
      uVRegion {bIsValid=false, max=<0 0>, min=<0 0>}
     }
     horizontalTopSlotImage {
      bIsDynamicallyLoaded=false
      drawAs=NoDrawType
      imageType=NoImage
      mirroring=NoMirror
      resourceName=None
      resourceObject=null
      tiling=NoTile
      imageSize=<0 0>
      margin=<ltrb 0 0 0 0>
      outlineSettings {
       bUseBrushTransparency=false
       roundingType=HalfHeightRadius
       width=0
       color {colorUseRule=UseColor_Specified, specifiedColor=<rgba 0 0 0 0>}
       cornerRadii=<0 0 0 0>
      }
      tintColor {colorUseRule=UseColor_Specified, specifiedColor=<rgba 1 1 1 1>}
      uVRegion {bIsValid=false, max=<0 0>, min=<0 0>}
     }
     hoveredThumbImage {
      bIsDynamicallyLoaded=false
      drawAs=RoundedBox
      imageType=NoImage
      mirroring=NoMirror
      resourceName=None
      resourceObject=null
      tiling=NoTile
      imageSize=<0 0>
      margin=<ltrb 0 0 0 0>
      outlineSettings {
       bUseBrushTransparency=false
       roundingType=FixedRadius
       width=0
       color {colorUseRule=UseColor_Specified, specifiedColor=<rgba 0 0 0 0>}
       cornerRadii=<4 4 4 4>
      }
      tintColor {colorUseRule=UseColor_Specified, specifiedColor=<rgba 0.2159 0.2159 0.2159 1>}
      uVRegion {bIsValid=false, max=<0 0>, min=<0 0>}
     }
     normalThumbImage {
      bIsDynamicallyLoaded=false
      drawAs=RoundedBox
      imageType=NoImage
      mirroring=NoMirror
      resourceName=None
      resourceObject=null
      tiling=NoTile
      imageSize=<0 0>
      margin=<ltrb 0 0 0 0>
      outlineSettings {
       bUseBrushTransparency=false
       roundingType=FixedRadius
       width=0
       color {colorUseRule=UseColor_Specified, specifiedColor=<rgba 0 0 0 0>}
       cornerRadii=<4 4 4 4>
      }
      tintColor {colorUseRule=UseColor_Specified, specifiedColor=<rgba 0.0953 0.0953 0.0953 1>}
      uVRegion {bIsValid=false, max=<0 0>, min=<0 0>}
     }
     verticalBackgroundImage {
      bIsDynamicallyLoaded=false
      drawAs=NoDrawType
      imageType=NoImage
      mirroring=NoMirror
      resourceName=None
      resourceObject=null
      tiling=NoTile
      imageSize=<0 0>
      margin=<ltrb 0 0 0 0>
      outlineSettings {
       bUseBrushTransparency=false
       roundingType=HalfHeightRadius
       width=0
       color {colorUseRule=UseColor_Specified, specifiedColor=<rgba 0 0 0 0>}
       cornerRadii=<0 0 0 0>
      }
      tintColor {colorUseRule=UseColor_Specified, specifiedColor=<rgba 1 1 1 1>}
      uVRegion {bIsValid=false, max=<0 0>, min=<0 0>}
     }
     verticalBottomSlotImage {
      bIsDynamicallyLoaded=false
      drawAs=NoDrawType
      imageType=NoImage
      mirroring=NoMirror
      resourceName=None
      resourceObject=null
      tiling=NoTile
      imageSize=<0 0>
      margin=<ltrb 0 0 0 0>
      outlineSettings {
       bUseBrushTransparency=false
       roundingType=HalfHeightRadius
       width=0
       color {colorUseRule=UseColor_Specified, specifiedColor=<rgba 0 0 0 0>}
       cornerRadii=<0 0 0 0>
      }
      tintColor {colorUseRule=UseColor_Specified, specifiedColor=<rgba 1 1 1 1>}
      uVRegion {bIsValid=false, max=<0 0>, min=<0 0>}
     }
     verticalTopSlotImage {
      bIsDynamicallyLoaded=false
      drawAs=NoDrawType
      imageType=NoImage
      mirroring=NoMirror
      resourceName=None
      resourceObject=null
      tiling=NoTile
      imageSize=<0 0>
      margin=<ltrb 0 0 0 0>
      outlineSettings {
       bUseBrushTransparency=false
       roundingType=HalfHeightRadius
       width=0
       color {colorUseRule=UseColor_Specified, specifiedColor=<rgba 0 0 0 0>}
       cornerRadii=<0 0 0 0>
      }
      tintColor {colorUseRule=UseColor_Specified, specifiedColor=<rgba 1 1 1 1>}
      uVRegion {bIsValid=false, max=<0 0>, min=<0 0>}
     }
    }
    widgetStyle {
     barThickness=8
     bottomShadowBrush {
      bIsDynamicallyLoaded=false
      drawAs=Box
      imageType=FullColor
      mirroring=NoMirror
      resourceName=../../../Engine/Content/Slate/Common/ScrollBoxShadowBottom.png
      resourceObject=null
      tiling=NoTile
      imageSize=<16 8>
      margin=<ltrb 0.5 0 0.5 1>
      outlineSettings {
       bUseBrushTransparency=false
       roundingType=HalfHeightRadius
       width=0
       color {colorUseRule=UseColor_Specified, specifiedColor=<rgba 0 0 0 0>}
       cornerRadii=<0 0 0 0>
      }
      tintColor {colorUseRule=UseColor_Specified, specifiedColor=<rgba 1 1 1 1>}
      uVRegion {bIsValid=false, max=<0 0>, min=<0 0>}
     }
     horizontalScrolledContentPadding=<ltrb 0 0 1 0>
     leftShadowBrush {
      bIsDynamicallyLoaded=false
      drawAs=Box
      imageType=FullColor
      mirroring=NoMirror
      resourceName=../../../Engine/Content/Slate/Common/ScrollBoxShadowLeft.png
      resourceObject=null
      tiling=NoTile
      imageSize=<8 16>
      margin=<ltrb 1 0.5 0 0.5>
      outlineSettings {
       bUseBrushTransparency=false
       roundingType=HalfHeightRadius
       width=0
       color {colorUseRule=UseColor_Specified, specifiedColor=<rgba 0 0 0 0>}
       cornerRadii=<0 0 0 0>
      }
      tintColor {colorUseRule=UseColor_Specified, specifiedColor=<rgba 1 1 1 1>}
      uVRegion {bIsValid=false, max=<0 0>, min=<0 0>}
     }
     rightShadowBrush {
      bIsDynamicallyLoaded=false
      drawAs=Box
      imageType=FullColor
      mirroring=NoMirror
      resourceName=../../../Engine/Content/Slate/Common/ScrollBoxShadowRight.png
      resourceObject=null
      tiling=NoTile
      imageSize=<8 16>
      margin=<ltrb 0 0.5 1 0.5>
      outlineSettings {
       bUseBrushTransparency=false
       roundingType=HalfHeightRadius
       width=0
       color {colorUseRule=UseColor_Specified, specifiedColor=<rgba 0 0 0 0>}
       cornerRadii=<0 0 0 0>
      }
      tintColor {colorUseRule=UseColor_Specified, specifiedColor=<rgba 1 1 1 1>}
      uVRegion {bIsValid=false, max=<0 0>, min=<0 0>}
     }
     topShadowBrush {
      bIsDynamicallyLoaded=false
      drawAs=Box
      imageType=FullColor
      mirroring=NoMirror
      resourceName=../../../Engine/Content/Slate/Common/ScrollBoxShadowTop.png
      resourceObject=null
      tiling=NoTile
      imageSize=<16 8>
      margin=<ltrb 0.5 1 0.5 0>
      outlineSettings {
       bUseBrushTransparency=false
       roundingType=HalfHeightRadius
       width=0
       color {colorUseRule=UseColor_Specified, specifiedColor=<rgba 0 0 0 0>}
       cornerRadii=<0 0 0 0>
      }
      tintColor {colorUseRule=UseColor_Specified, specifiedColor=<rgba 1 1 1 1>}
      uVRegion {bIsValid=false, max=<0 0>, min=<0 0>}
     }
     verticalScrolledContentPadding=<ltrb 0 0 0 1>
    }
   }
   slot {class=BorderSlot, properties {padding=<ltrb 8 8 8 8>}}
  }
  VerticalBox {
   widgetClassPath=&VerticalBox
   bIsVariable=true
   parent=ScrollBox_209
   widget=VerticalBox
   slot {class=ScrollBoxSlot, properties {size {sizeRule=Fill, value=1}}}
  }
 ]
}
